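# Sound cue for the start of the Fire of the Dragon slash, a melee weapon sound under the gadget audio folder.
# A Random node picks between two equally weighted start waves and feeds a root Modulator that varies playback, with its minimum pitch held at unaltered.
# The cue uses small melee weapon attenuation and the melee weapon sound class, with a boosted volume multiplier.
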SoundCue /Game/Sounds/Fort_Gadgets/Dragon_Slash/WPN_Fire_ofthe_Dragon_Slash_Start_Cue
root=Modulator_0
nodes [
 {id=Modulator_0, children [Random_1], properties {pitchMin=1}}
 {id=Random_1, children [WavePlayer_2,WavePlayer_3], properties {weights [1,1]}}
 {id=WavePlayer_2, properties {soundWaveAssetPtr=&/Fire_of_the_Dragon_Start_01}}
 {id=WavePlayer_3, properties {soundWaveAssetPtr=&/Fire_of_the_Dragon_Start_02}}
]
properties {
 attenuationSettings=&Fort_Weapon_Melee_Sml_Att
 maxDistance=2800
 soundClassObject=&Fort_WeaponMelee_PSM
 volumeMultiplier=1.25
}
